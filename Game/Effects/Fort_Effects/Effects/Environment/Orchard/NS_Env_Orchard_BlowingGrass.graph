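# Niagara system for wind-blown grass in the Orchard environment, using one GPU sprite emitter in intermittent gusts with randomized delays.
# Each gust spawns sprites in a thin band that sweeps along X over the loop, with landscape sampling supplying terrain offset.
# Initial velocity, curl noise, drag and spin drive the motion.
# Time-of-day parameters set color and a material parameter, while camera distance scales sprite size and alpha fades over lifetime.
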NiagaraSystem /Game/Effects/Fort_Effects/Effects/Environment/Orchard/NS_Env_Orchard_BlowingGrass
dependencies {
 usedDataInterfaces [
  &NiagaraDataInterfaceCamera
  &NiagaraDataInterfaceCurve
  &NiagaraDataInterfaceGBuffer
  &NiagaraDataInterfaceColorCurve
 ]
 usedDynamicInputs [&SimulationPosition]
 usedModules [
  &SystemState
  &self:SystemSpawnScript.SystemScriptSource.SystemScriptGraph.NiagaraNodeAssignment_0.SetVariables_#1
 ]
}
emitters {Dust_Motes {
  simTarget=GPUComputeSim
  emitterUpdateScript [
   /SetVariables(Emitter.CameraPostion:Vector3f=&ReturnCameraVector(Camera Property=::NewEnumerator0, Coordinate Space=::Simulation)) {isSetParameters=true}
   &EmitterState(
    Life Cycle Mode=::NewEnumerator1
    Loop Delay=&RandomRangeFloat(Minimum=5, Maximum=10)
    Loop Duration=&RandomRangeFloat(Minimum=1, Maximum=2)
    Recalculate Duration Each Loop=true
    UseLoopDelay=true
   )
   &SpawnRate(SpawnRate=&Lerp_Float(A=0, B=100, Alpha=$NPC.FortniteNPC.IsPlayerIndoors), Use Spawn Probability=false) {enabled=false}
   &SpawnRate(SpawnRate=500, Use Spawn Probability=false) {module=SpawnRate001}
  ]
  particleSpawnScript [
   /SetVariables(Particles.RandSpreadCurve:NiagaraFloat=&FloatFromCurve(
     FloatCurve=DataInterface() {propertyValues {Curve {keys [
         {interpMode=RCIM_Cubic, tangentMode=RCTM_User, time=0, value=0.7}
         {interpMode=RCIM_Cubic, tangentMode=RCTM_User, time=0.8, value=0.4}
         {arriveTangent=2.7919, interpMode=RCIM_Cubic, leaveTangent=2.7919, tangentMode=RCTM_User, time=1, value=1}
        ]}}}
     CurveIndex=&RandomRangeFloat(Minimum=0, Maximum=1)
    )) {isSetParameters=true}
   &InitializeParticle(
    Color=<rgba 0 0 0 1>
    Lifetime=&UniformRangedFloat(Minimum=1, Maximum=3)
    Mass Mode=::NewEnumerator1
    Sprite Rotation Mode=::NewEnumerator2
    Sprite Size Min=<2 5>
    Sprite Size Mode=::NewEnumerator2
    UsePositionOffset=false
    Write Material Random=true
   ) {module=InitializeParticle001}
   &ShapeLocation(
    MaskSpawnGroup=false
    Non Uniform Scale=<0.02 1 0.02>
    Offset=&MakeVector(
     X=&FloatFromCurve(
      FloatCurve=DataInterface() {propertyValues {Curve {keys [<at 0 1>, {arriveTangent=-2.8236, leaveTangent=-2.8236, time=1, value=-1}]}}}
      CurveIndex=$Emitter.NormalizedLoopAge
      Scale Curve=-1000
     )
     Y=0
     Z=0
    )
    Offset Mode=::NewEnumerator0
    Override Seed=false
    Rotation Mode=::NewEnumerator1
    Sphere Radius=2500
    UseOwnerScale=true
   )
   /SetVariables(Particles.Terrain Offset=5) {isSetParameters=true}
   &self:Sample_Landscape() {module=Sample_Landscape}
   /SetVariables(Particles.Position=&Add_Vector(
     A=Linked() {linkedVariable {name=Particles.Position, type=NiagaraPosition}}
     B=&MakeVector(X=0, Y=0, Z=$Particles.Terrain Offset)
    )) {isSetParameters=true}
   &AddVelocity(Velocity=<1500 0 0>)
   /Niagara/Modules/Update/Camera/MaintainInCameraParticleScale()
   &DynamicMaterialParameters(
    Index 0 Param 1=0.25
    Index 0 Param 3=150
    Index 1 Param 1=1
    Module.Param0WriteEnabled=true
    Module.Param0WriteEnabled1=true
    Module.Param1WriteEnabled=true
    Module.Param1WriteEnabled1=true
    Module.Param2WriteEnabled=true
    Module.Param2WriteEnabled1=true
    Module.Param3WriteEnabled=true
    Module.Param3WriteEnabled1=true
    Write Parameter Index 1=true
   ) {module=DynamicMaterialParameters001}
   /Game/Effects/Niagara/Modules/Spawn/FN_Spawn_GBuffer_Color() {enabled=false}
   &Color(Color=&MultiplyLinearColor(
     Linear Color=&ColorFromCurve(
      ColorCurve=DataInterface() {propertyValues {
        AlphaCurve {keys [<at 0 1 RCIM_Cubic>,<at 1 1 RCIM_Cubic>]}
        BlueCurve {keys [<at 0 0.1029 RCIM_Cubic>,<at 0.25 0.004>,<at 0.5853 0.0123>,<at 1 0.1029 RCIM_Cubic>]}
        GreenCurve {keys [<at 0 0.4115 RCIM_Cubic>,<at 0.25 0.0242>,<at 0.5853 0.1356>,<at 1 0.4115 RCIM_Cubic>]}
        RedCurve {keys [<at 0 0.1335 RCIM_Cubic>,<at 0.25 0.0103>,<at 0.5853 0.1651>,<at 1 0.1335 RCIM_Cubic>]}
       }}
      CurveIndex=Linked() {linkedVariable {name=NPC.FortniteNPC.TimeOfDay_0-1, type=NiagaraFloat}}
     )
     Scale Factor=&MakeVector4_FromFloat(Float=1)
    ))
  ]
  particleUpdateScript [
   &self:Sample_Landscape_UPdate(Bias:NiagaraFloat=$Particles.Terrain Offset, New Landscape Sample:NiagaraDataInterfaceLandscape=$User.Landscape) {module=Sample_Landscape_UPdate}
   &ParticleState()
   &SpriteRotationRate(Rotation Rate=250) {module=SpriteRotationRate001}
   &self:Camera_to_Particle_Distance_Calculation_01(Camera Postion:Vector3f=$Emitter.CameraPostion, Distance=1000, Particle Position:NiagaraPosition=$Particles.Position) {module=Camera_to_Particle_Distance_Calculation_01}
   &ScaleSpriteSize(Scale Factor=&Vector2DFromFloat(Value=&FloatFromCurve(
      FloatCurve=DataInterface() {propertyValues {Curve {keys [
          {interpMode=RCIM_Cubic, tangentMode=RCTM_User, time=0, value=1}
          {arriveTangent=3.3998, interpMode=RCIM_Cubic, leaveTangent=3.3998, tangentMode=RCTM_User, time=1, value=2}
         ]}}}
      CurveIndex=$Particles.NormalizedDistance
     )))
   &SpriteRotationRate(Rotation Rate=&RandomRangeFloat(Minimum=-45, Maximum=150))
   &ScaleColor(
    Scale Alpha=&FloatFromCurve(
     FloatCurve=DataInterface() {propertyValues {Curve {keys [
         {interpMode=RCIM_Cubic, tangentMode=RCTM_User, time=0, value=0}
         {arriveTangent=1.8182, leaveTangent=1.8182, time=0.35, value=1}
         <at 0.55 1>
         <at 1 0>
        ]}}}
     Scale Curve=5
    )
    Scale RGB=&Multiply_VectorByFloat(Vector=<1 1 1>, Float=1)
    ScaleA=true
    ScaleRGB=true
   )
   &GravityForce(Gravity=<50 0 0>)
   &CurlNoiseForce(
    Noise Frequency=&RandomRangeFloat(Minimum=1, Maximum=5)
    Noise Strength=&RandomRangeFloat(Minimum=250, Maximum=300)
    Pan Noise=true
    Pan Noise Field=<1 0 0>
   )
   &Drag(
    Drag=&FloatFromCurve(FloatCurve=DataInterface() {propertyValues {Curve {keys [
         {arriveTangent=-0.6113, interpMode=RCIM_Cubic, leaveTangent=-0.6113, tangentMode=RCTM_User, time=0, value=0}
         {arriveTangent=-2.5249, interpMode=RCIM_Cubic, leaveTangent=-2.5249, tangentMode=RCTM_User, time=0.1408, value=0.9847}
         {arriveTangent=-0.3425, interpMode=RCIM_Cubic, leaveTangent=-0.3425, tangentMode=RCTM_User, time=0.3, value=0.15}
         {arriveTangent=-0.0405, interpMode=RCIM_Cubic, leaveTangent=-0.0405, tangentMode=RCTM_User, time=1, value=0}
        ]}}})
    Use Linear Drag=true
    Use Rotational Drag=true
   )
   &SolveForcesAndVelocity(Clamp Velocity=true, Limit Acceleration=false, Speed Limit=250)
   &DynamicMaterialParameters(
    Index 0 Param 1=&RandomRangeFloat(Minimum=-0.05, Maximum=0.05)
    Index 0 Param 2=&FloatFromCurve(
     FloatCurve=DataInterface() {propertyValues {Curve {keys [<at 0 0>,<at 0.2 0.8>,<at 0.8 0>,<at 1 0>]}}}
     CurveIndex=Linked() {linkedVariable {name=NPC.FortniteNPC.TimeOfDay_0-1, type=NiagaraFloat}}
    )
    Index 0 Param 3=1
    Module.Index0Param0Write=true
    Module.Index0Param1Write=true
    Module.Index0Param2Write=true
    Module.Index0Param3Write=true
   )
  ]
  rendererClasses [&NiagaraSpriteRendererProperties]
  renderers [{rendererClass=&NiagaraSpriteRendererProperties}]
 }}
summary {
 systemName=NS_Env_Orchard_BlowingGrass
 emitters [{emitterName=Dust_Motes, simTarget=GPUComputeSim, rendererClasses [&NiagaraSpriteRendererProperties]}]
 userVariables [User.Landscape {
   type=NiagaraDataInterfaceLandscape
   description=""
   defaultValue {
    struct=NiagaraExt_VariableValue_DataInterface
    value {
     dataInterface=&self:NiagaraDataInterfaceLandscape_0
     dataInterfaceClass=&NiagaraDataInterfaceLandscape
    }
   }
  }]
}
