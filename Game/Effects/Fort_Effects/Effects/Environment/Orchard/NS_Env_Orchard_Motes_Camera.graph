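# Orchard ambient Niagara system of GPU sprite dust motes, with one continuously spawning emitter and two sprite renderers.
# Particles spawn across a wide flattened sphere, take their color from time of day, and sample the landscape with a height offset.
# Curl noise and gravity force drive speed-capped drift, while waveform and curves flicker and fade alpha.
# Camera distance scales sprite size up, and an indoor-driven spawn rate is present but disabled.
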NiagaraSystem /Game/Effects/Fort_Effects/Effects/Environment/Orchard/NS_Env_Orchard_Motes_Camera
dependencies {
 usedDataInterfaces [&NiagaraDataInterfaceCamera, &NiagaraDataInterfaceCurve, &NiagaraDataInterfaceColorCurve]
 usedDynamicInputs [&SimulationPosition]
 usedModules [
  &SystemState
  &self:SystemSpawnScript.SystemScriptSource.SystemScriptGraph.NiagaraNodeAssignment_0.SetVariables_#1
 ]
}
emitters {Dust_Motes {
  simTarget=GPUComputeSim
  emitterUpdateScript [
   /SetVariables(Emitter.CameraPostion:Vector3f=&ReturnCameraVector(Camera Property=::NewEnumerator0, Coordinate Space=::Simulation)) {isSetParameters=true}
   &EmitterState(
    Life Cycle Mode=::NewEnumerator1
    Loop Duration=&RandomRangeFloat(Minimum=5, Maximum=10)
    Recalculate Duration Each Loop=true
    UseLoopDelay=false
   )
   &SpawnRate(SpawnRate=&Lerp_Float(A=0, B=100, Alpha=$NPC.FortniteNPC.IsPlayerIndoors), Use Spawn Probability=false) {enabled=false}
   &SpawnRate(SpawnRate=250, Use Spawn Probability=false) {module=SpawnRate001}
  ]
  particleSpawnScript [
   /SetVariables(Particles.RandSpreadCurve:NiagaraFloat=&FloatFromCurve(
     FloatCurve=DataInterface() {propertyValues {Curve {keys [
         {interpMode=RCIM_Cubic, tangentMode=RCTM_User, time=0, value=0.7}
         {interpMode=RCIM_Cubic, tangentMode=RCTM_User, time=0.8, value=0.4}
         {arriveTangent=2.7919, interpMode=RCIM_Cubic, leaveTangent=2.7919, tangentMode=RCTM_User, time=1, value=1}
        ]}}}
     CurveIndex=&RandomRangeFloat(Minimum=0, Maximum=1)
    )) {isSetParameters=true}
   &InitializeParticle(
    Color=&MakeLinearColorFromVectorAndFloat(
     "Vector (RGB)"=&MakeVectorFromLinearColorRGB(Color=&MultiplyLinearColor(
       Linear Color=&ColorFromCurve(
        ColorCurve=DataInterface() {propertyValues {
          AlphaCurve {keys []}
          BlueCurve {keys [<at 0.1189 1>,<at 0.1666 0>,<at 0.3109 0.4062>,<at 0.5746 1>,<at 0.8383 0>,<at 0.875 0>,<at 0.9354 1>,<at 1 1>]}
          GreenCurve {keys [<at 0.1189 0.5833>,<at 0.1666 0.1067>,<at 0.3109 0.9>,<at 0.5746 1>,<at 0.8383 0.5>,<at 0.875 0.2>,<at 0.9354 0.2942>,<at 1 0.625>]}
          RedCurve {keys [<at 0.1189 0.5>,<at 0.1666 1>,<at 0.3109 1>,<at 0.5746 1>,<at 0.8383 1>,<at 0.875 1>,<at 0.9354 0.23>,<at 1 0.5>]}
         }}
        CurveIndex=Linked() {linkedVariable {name=NPC.FortniteNPC.TimeOfDay_0-1, type=NiagaraFloat}}
       )
       Scale Factor=&MakeVector4_FromFloat(Float=1)
      ))
     "Float (Alpha)"=&RandomRangeFloat(Minimum=1, Maximum=1)
    )
    Lifetime=&UniformRangedFloat(Minimum=4, Maximum=6)
    Mass Mode=::NewEnumerator1
    Sprite Rotation Mode=::NewEnumerator2
    Sprite Size=&Vector2DFromFloat(Value=&RandomRangeFloat(Minimum=0.5, Maximum=1))
    Sprite Size Mode=::NewEnumerator4
    UsePositionOffset=false
    Write Material Random=true
   ) {module=InitializeParticle001}
   /SetVariables(Particles.Terrain Offset=15) {isSetParameters=true}
   &self:Sample_Landscape() {module=Sample_Landscape}
   /SetVariables(Particles.Position=&Add_Vector(
     A=Linked() {linkedVariable {name=Particles.Position, type=NiagaraPosition}}
     B=&MakeVector(X=0, Y=0, Z=$Particles.Terrain Offset)
    )) {isSetParameters=true}
   &ShapeLocation(
    MaskSpawnGroup=false
    Non Uniform Scale=<0.1 1 1>
    Override Seed=false
    Rotation Mode=::NewEnumerator1
    Sphere Radius=2500
    UseOwnerScale=true
   )
   &AddVelocity(Velocity=<50 0 0>)
   /Niagara/Modules/Update/Camera/MaintainInCameraParticleScale()
   &DynamicMaterialParameters(
    Index 0 Param 1=2
    Index 0 Param 2=50
    Index 0 Param 3=150
    Index 1 Param 1=1
    Module.Param0WriteEnabled=true
    Module.Param0WriteEnabled1=true
    Module.Param1WriteEnabled=true
    Module.Param1WriteEnabled1=true
    Module.Param2WriteEnabled=true
    Module.Param2WriteEnabled1=true
    Module.Param3WriteEnabled=true
    Module.Param3WriteEnabled1=true
    Write Parameter Index 1=true
   ) {module=DynamicMaterialParameters001}
  ]
  particleUpdateScript [
   &self:Sample_Landscape_UPdate(Bias:NiagaraFloat=$Particles.Terrain Offset, New Landscape Sample:NiagaraDataInterfaceLandscape=$User.Landscape) {module=Sample_Landscape_UPdate}
   &ParticleState()
   &SpriteRotationRate(Rotation Rate=&RandomRangeFloat(Minimum=-250, Maximum=250)) {module=SpriteRotationRate001}
   &self:Camera_to_Particle_Distance_Calculation(Camera Postion:Vector3f=$Emitter.CameraPostion, Distance=1000, Particle Position:NiagaraPosition=$Particles.Position) {module=Camera_to_Particle_Distance_Calculation}
   &ScaleSpriteSize(Scale Factor=&Vector2DFromFloat(Value=&FloatFromCurve(
      FloatCurve=DataInterface() {propertyValues {Curve {keys [
          {interpMode=RCIM_Cubic, tangentMode=RCTM_User, time=0, value=1}
          {arriveTangent=3.1382, interpMode=RCIM_Cubic, leaveTangent=3.1382, tangentMode=RCTM_User, time=0.8, value=2}
          <at 1 5 RCIM_Cubic>
         ]}}}
      CurveIndex=$Particles.NormalizedDistance
     )))
   &SpriteRotationRate(Rotation Rate=&RandomRangeFloat(Minimum=-45, Maximum=45))
   &ScaleColor(
    Scale Alpha=&Multiply_Float(
     A=&Lerp_Float(A=1, B=0.3, Alpha=$Particles.RandSpreadCurve)
     B=&Waveform(
      "Waveform [1]"=::NewEnumerator3
      "[1] Frequency"=&RandomRangeFloat(Minimum=0.2, Maximum=0.5)
      "[1] Amplitude Min/Max"=<0.5 1>
     )
    )
    Scale RGB=&VectorFromFloat(Value=&Lerp_Float(A=1, B=1, Alpha=$Particles.RandSpreadCurve))
    ScaleA=true
    ScaleRGB=true
   ) {module=ScaleColor001}
   &ScaleColor(
    Scale Alpha=&FloatFromCurve(
     FloatCurve=DataInterface() {propertyValues {Curve {keys [
         {interpMode=RCIM_Cubic, tangentMode=RCTM_User, time=0, value=0}
         {arriveTangent=1.8182, leaveTangent=1.8182, time=0.35, value=1}
         <at 0.55 1>
         <at 1 0>
        ]}}}
     Scale Curve=0.5
    )
    Scale RGB=&Multiply_VectorByFloat(Vector=<5 5 5>, Float=1)
    ScaleA=true
    ScaleRGB=true
   )
   &GravityForce(Gravity=<50 0 50>)
   &CurlNoiseForce(
    Curl Noise Cone Mask Angle=60
    Mask Curl Noise=true
    Noise Frequency=&RandomRangeFloat(Minimum=1, Maximum=2)
    Noise Strength=&RandomRangeFloat(Minimum=80, Maximum=800)
    Pan Noise=true
    Pan Noise Field=<1 0 0>
   )
   &SolveForcesAndVelocity(Clamp Velocity=true, Limit Acceleration=false, Speed Limit=85)
   &DynamicMaterialParameters(
    Index 0 Param 1=1
    Index 0 Param 2=150
    Index 0 Param 3=150
    Index 1 Param 1=1
    Index 1 Param 2=1
    Module.Index0Param0Write=true
    Module.Index0Param1Write=true
    Module.Index0Param2Write=true
    Module.Index0Param3Write=true
    Module.Index1Param0Write=true
    Module.Index1Param1Write=true
    Module.Index1Param2Write=false
    Module.Index1Param3Write=true
    Write Parameter Index 1=true
   )
  ]
  rendererClasses [&NiagaraSpriteRendererProperties]
  renderers [
   {rendererClass=&NiagaraSpriteRendererProperties}
   {rendererClass=&NiagaraSpriteRendererProperties, rendererIndex=1}
  ]
 }}
summary {
 systemName=NS_Env_Orchard_Motes_Camera
 emitters [{emitterName=Dust_Motes, simTarget=GPUComputeSim, rendererClasses [&NiagaraSpriteRendererProperties]}]
 userVariables [User.Landscape {
   type=NiagaraDataInterfaceLandscape
   description=""
   defaultValue {
    struct=NiagaraExt_VariableValue_DataInterface
    value {
     dataInterface=&self:NiagaraDataInterfaceLandscape_0
     dataInterfaceClass=&NiagaraDataInterfaceLandscape
    }
   }
  }]
}
